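# Surface material for the Insectoid vehicle wheel, adding animated energy-like patterns to parts of the rim and tire.
# It extends MF_WheelParent_Simple by summing rotating ray and panning distortion and swirl textures, then multiplying that into base color and emissive.
# Rim green and tire red mask channels confine the effect, with vertex color red separating rim from tire.
# The result passes through the shared vehicle damage and de-rez transition functions.
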Material /VehicleCosmetics/Wheels/Insectoid/Materials/M_Wheel_Insectoid
graph {
 properties {
  bAutomaticallySetUsageInEditor=false
  bUseMaterialAttributes=true
  bUsedWithMeshDeformer=true
  bUsedWithSkeletalMesh=true
  pixelDepthOffsetMode=PDOM_Legacy
  usedShadingModels=Default Lit
 }
 nodes [
  n1 = MaterialFunctionCall() {materialFunction=&MF_WheelParent_Simple}
  n2 = GetMaterialAttributes(None=@n1.MaterialAttributes) {attributeGetTypes [#1, #2]}
  n3 = Rotator()
  n4 = Panner() {speedY=-0.2}
  n5 = Add(
   A=Add(
    A=TextureSample(UVs=@n3) {texture=/VehicleCosmetics/Wheels/_SharedTextures/FX/RaysGradient_RGB_T}.R
    B=TextureSample(UVs=@n4) {texture=/VehicleCosmetics/Wheels/_SharedTextures/FX/Distort_RGB_T}.B
   )
   B=TextureSample(UVs=@n4) {samplerType=SAMPLERTYPE_LinearColor, texture=/VehicleCosmetics/Wheels/Insectoid/Textures/Swirl_RGB_02_T}.R
  )
  n6 = Multiply(A=@n2.Base Color, B=@n5)
  n7 = Multiply(A=ComponentMask(None=@n1.RimRGBAMask) {g=true}, B=OneMinus(None=@n1.VertexColorR))
  n8 = Multiply(A=ComponentMask(None=@n1.TireRGBAMask) {r=true}, B=@n1.VertexColorR)
  n9 = LinearInterpolate(A=LinearInterpolate(A=@n2.Base Color, B=@n6, Alpha=@n7), B=@n6, Alpha=@n8)
  n10 = Multiply(A=@n2.Emissive Color, B=@n5)
  n11 = LinearInterpolate(A=LinearInterpolate(A=@n2.Emissive Color, B=@n10, Alpha=@n7), B=@n10, Alpha=@n8)
  MaterialFunctionCall(In=MaterialFunctionCall(Material=SetMaterialAttributes!(MaterialAttributes=@n2.MaterialAttributes, Base Color=@n9, Emissive Color=@n11) {attributeSetTypes [#1, #2]}) {materialFunction=&MF_BaseVehicle_Damage}.Result) {materialFunction=&MF_VehicleDeRez_Transition}
 ]
}
